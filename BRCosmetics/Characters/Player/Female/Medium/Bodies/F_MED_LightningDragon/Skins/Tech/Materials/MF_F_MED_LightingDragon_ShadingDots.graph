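# Material function for the Lightning Dragon Tech character skin that adds stylized halftone dots and a tinted shaded shirt to emissive shading, behind a static switch.
# Shading masks and curve inputs, plus a directional-light Lambert term, support thresholded masks for blending material attributes.
# Dots use a rotated grid on a camera-facing virtual plane anchored near the actor, scaled by FOV, platform and camera distance.
# LUT-processed cloud noise is added before saturating emissive.
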MaterialFunction /BRCosmetics/Characters/Player/Female/Medium/Bodies/F_MED_LightningDragon/Skins/Tech/Materials/MF_F_MED_LightingDragon_ShadingDots
graph {nodes [
  FunctionOutput(None=StaticSwitchParameter(True=NamedRerouteUsage(), False=NamedRerouteUsage()) {group=1_LightingDragon_ShadingDots, parameterName=UseLightingDragon_ShadingDots, sortPriority=0})
  n1 = NamedRerouteUsage()
  n2 = NamedRerouteUsage()
  n3 = VectorParameter() {group=1_LightingDragon_ShadingDots, parameterName=FishScaleShirt_ShadedColor, defaultValue=<rgba 0 0 0 1>}
  n4 = NamedRerouteUsage()
  n5 = ScalarParameter() {group=1_LightingDragon_ShadingDots, parameterName=FishScaleShirt_Shading_Offset, sliderMax=0.5, sliderMin=-0.5, sortPriority=20}
  n6 = ScalarParameter() {defaultValue=6, group=1_LightingDragon_ShadingDots, parameterName=FishScaleShirt_Shading_Transition, sliderMax=2, sortPriority=20}
  n7 = Saturate(None=Multiply(A=Add(A=@n4, B=@n5) {constB=-0.25}, B=@n6) {constB=1.5})
  n8 = BlendMaterialAttributes(
   A=@n2
   B=SetMaterialAttributes!(MaterialAttributes=@n2, EmissiveColor=@n3.RGB) {attributeSetTypes [#1]}
   Alpha=Multiply(A=Multiply(A=@n3.A, B=@n7), B=NamedRerouteUsage())
  )
  n9 = Multiply(A=NamedRerouteUsage(), B=NamedRerouteUsage())
  n10 = NamedRerouteUsage()
  n11 = GetMaterialAttributes(None=BlendMaterialAttributes(
    A=@n8
    B=SetMaterialAttributes(MaterialAttributes=@n8, EmissiveColor=@n1) {attributeSetTypes [#1]}
    Alpha=Multiply(A=@n9, B=@n10)
   )) {attributeGetTypes [#1]}
  n12 = TextureSample(UVs=Multiply(A=NamedRerouteUsage()) {constB=0.5}) {texture=&AtmosphericCloudNoise01}
  n13 = ScalarParameter() {defaultValue=0.01, group=1_LightingDragon_ShadingDots, parameterName=ScreenSpace_Noise_Intensity}
  n14 = Add(
   A=@n11.EmissiveColor
   B=MaterialFunctionCall!(In=Multiply(A=@n12.R, B=@n13) {constB=0.02}, LUT, LUT Multiplier) {materialFunction=/Game/Packages/Fortress_SharedMaterials/Anime_Master_Material/MF_ApplyLUT}.Result
  )
  NamedRerouteDeclaration(None=SetMaterialAttributes(MaterialAttributes=@n11.MaterialAttributes, EmissiveColor=Saturate(None=@n14)) {attributeSetTypes [#1]}) {name=MA_Result, nodeColor=<rgba 1 0.7736 0 1>}
  NamedRerouteDeclaration(None=FunctionInput() {bUsePreviewValueAsDefault=true, inputName=ShadingCurve, inputType=FunctionInput_Scalar, sortPriority=1}) {name=ShadingCurve, nodeColor=<rgba 1 0 0.6562 1>}
  n15 = MaterialFunctionCall(Float3=FunctionInput() {inputName=FX, inputType=FunctionInput_Vector4, sortPriority=4}) {materialFunction=&BreakOutFloat3Components}
  NamedRerouteDeclaration(None=@n15.B) {name=ShirtMask, nodeColor=<rgba 0.6406 0 1 1>}
  n16 = MaterialFunctionCall() {materialFunction=&CameraDirectionVector}
  n17 = Multiply(A=Constant3Vector() {constant=<rgba 1 1 0 0>}, B=@n16.Result)
  n18 = Add(A=Multiply(B=Normalize(VectorInput=@n17)) {constA=0.25}, B=@n16.Result)
  n19 = Multiply(A=@n16.Result) {constB=50}
  n20 = Add(A=ActorPositionWS(), B=Constant3Vector() {constant=<rgba 0 0 90 1>})
  n21 = LinearInterpolate(
   A=Constant() {r=64}
   B=Constant() {r=128}
   Alpha=InverseLinearInterpolate(
    A=Constant() {r=45}
    B=Constant() {r=90}
    Value=MaterialFunctionCall() {materialFunction=/Engine/Functions/Engine_MaterialFunctions02/Utility/FOV}.FOV angle in degrees
   )
  )
  n22 = Multiply(A=@n21) {constB=1.5}
  NamedRerouteDeclaration(None=MaterialFunctionCall(Plane Normal Axis=Normalize(VectorInput=@n18), Plane Center=Add(A=@n19, B=@n20), UV Scale=ShadingPathSwitch(Default=@n21, Mobile=@n22)) {materialFunction=&VirtualPlaneCoordinates}.UVs) {name=ObjectAlighedScreenPosition, nodeColor=<rgba 1 0.867 0 1>}
  n23 = DotProduct(A=Custom() {code="return View.DirectionalLightDirection;", description=DirectionalLightDirection, inputs []}, B=VertexNormalWS())
  n24 = Multiply(A=NamedRerouteDeclaration(None=Multiply(A=@n23) {constB=-1}) {name=LdotN, nodeColor=<rgba 0.2656 0 1 1>}) {constB=0.5}
  NamedRerouteDeclaration(None=Saturate(None=Add(A=@n24) {constB=0.5})) {name=Lambert, nodeColor=<rgba 0 0.9297 1 1>}
  n25 = Subtract(A=Distance(A=ActorPositionWS(), B=CameraPositionWS())) {constB=100}
  NamedRerouteDeclaration(None=Saturate(None=Divide(A=@n25) {constB=400})) {name=DistanceBlending, nodeColor=<rgba 0.4766 1 0 1>}
  NamedRerouteDeclaration(None=@n15.R) {name=ShadingMask, nodeColor=<rgba 0 0.2031 1 1>}
  n26 = ScalarParameter() {defaultValue=90, group=1_LightingDragon_ShadingDots, parameterName=Dots_Amount, sliderMax=512, sortPriority=20}
  n27 = Multiply(A=NamedRerouteUsage(), B=LinearInterpolate(A=@n26, B=Multiply(A=@n26) {constB=0.5}, Alpha=NamedRerouteUsage()))
  n28 = Constant() {r=1}
  n29 = Add(A=Frac(None=Rotator(Coordinate=@n27, Time=@n28) {speed=0.75})) {constB=-0.5}
  n30 = Constant()
  NamedRerouteDeclaration(None=Multiply(
    A=Distance(A=@n29, B=@n30)
    B=ScalarParameter() {defaultValue=1.25, group=1_LightingDragon_ShadingDots, parameterName=DotsGradient_Scale, sliderMax=4}
   ) {constB=4}) {name=DotsGradient, nodeColor=<rgba 1 0.0938 0 1>}
  n31 = NamedRerouteUsage()
  n32 = ScalarParameter() {defaultValue=0.5, group=1_LightingDragon_ShadingDots, parameterName=ShadingDots_Offset, sliderMax=0.5, sliderMin=-0.5, sortPriority=20}
  n33 = ScalarParameter() {defaultValue=2, group=1_LightingDragon_ShadingDots, parameterName=ShadingDots_Transition, sliderMax=2, sortPriority=20}
  n34 = NamedRerouteUsage()
  n35 = Add(A=Multiply(A=Add(A=@n31, B=@n32) {constB=-0.25}, B=@n33) {constB=1.5}, B=Multiply(A=@n33, B=OneMinus(None=@n34)) {constB=1.5})
  n36 = OneMinus(None=NamedRerouteUsage())
  n37 = Subtract(A=Saturate(None=Multiply(A=@n35, B=@n36)), B=NamedRerouteUsage())
  n38 = NamedRerouteDeclaration(None=SmoothStep(Min=Constant() {r=0.45}, Max=Constant() {r=0.55}, Value=Saturate(None=@n37))) {name=ShadingDotsMask, nodeColor=<rgba 0.4766 1 0 1>}
  n39 = VectorParameter() {group=1_LightingDragon_ShadingDots, parameterName=ShadingDots_Color, sortPriority=10, defaultValue=<rgba 0.1 0.1 0.1 1>}
  NamedRerouteDeclaration(None=AppendVector(A=Multiply(A=@n38, B=@n39.RGB), B=@n39.A)) {name=ShadingDots, nodeColor=<rgba 1 0 0.6797 1>}
  NamedRerouteDeclaration(None=@n15.G) {name=NegativeThreshold, nodeColor=<rgba 0 0.9297 1 1>}
  NamedRerouteDeclaration(None=FunctionInput() {inputName=MA, inputType=FunctionInput_MaterialAttributes}) {name=MA, nodeColor=<rgba 0.3594 1 0 1>}
  n40 = FunctionInput() {bUsePreviewValueAsDefault=true, inputName=SSC, sortPriority=3}
  NamedRerouteDeclaration(None=ComponentMask(None=@n40) {r=true}) {name=Threshold, nodeColor=<rgba 0 0.6953 1 1>}
  NamedRerouteDeclaration(None=FunctionInput() {bUsePreviewValueAsDefault=true, inputName=InkLine Mask, inputType=FunctionInput_Scalar, sortPriority=2}) {name=DF_Line_Mask, nodeColor=<rgba 1 0.3281 0 1>}
  NamedRerouteDeclaration(None=ComponentMask(None=@n40) {b=true}) {name=Sheen, nodeColor=<rgba 0 1 0.6484 1>}
 ]}
